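; Gameplay ability for the exotic Spicy ColdFern weapon that applies a "spicy" on-fire gameplay effect to the activating pawn.
; It activates from a gameplay event and applies the configured SpicyGameplayEffectClass from the pawn's ability system component to itself, then ends immediately.
; The event data is not used.
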(event Ability|EventActivateAbilityFromEvent
  (params (EventData :type Gameplay%20Event%20Data%20Structure%20%28by%20ref%29))
  (GameplayEffects|ApplyGameplayEffectToTarget :self (Ability|GetAbilitySystemComponent (Ability|GetActivatingPawn)) :GameplayEffectClass (Variables|Default|GetSpicyGameplayEffectClass) :Target (Ability|GetAbilitySystemComponent (Ability|GetActivatingPawn)) :Level 0.0)
  (Ability|EndAbility))
